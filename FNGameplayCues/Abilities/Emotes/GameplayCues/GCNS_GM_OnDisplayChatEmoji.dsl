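; Static gameplay cue for chat emoji display: plays a bubble sound and sets up the emoji particle for the viewed player.
; HandleGameplayCue plays the 2D sound only when the target pawn is locally viewed.
; On particle spawn, the source emoji definition configures the component with the instigator's quick chat icon if that pawn is player 0's.
; A live broadcast spectator fallback matches its controller's pawn and adds a relative offset; otherwise the particle is deactivated.
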(fn HandleGameplayCue
  (params (MyTarget :type Actor%20Object%20Reference) (EventType :type EGameplayCueEvent%20Enum :default "OnActive") (Parameters :type Gameplay%20Cue%20Parameters%20Structure%20%28by%20ref%29))
  (bind _asfort_player_pawn (Utilities|Casting|CastToFortPlayerPawn MyTarget))
  (if (Pawn|IsLocallyViewed _asfort_player_pawn)
    (Audio|SpawnSound2D MyTarget (Variables|Default|GetBubbleSound))
    (return)))

(fn OnActive
  (params (MyTarget :type Actor%20Object%20Reference) (Parameters :type Gameplay%20Cue%20Parameters%20Structure%20%28by%20ref%29))
  (return (|CallParentFunction MyTarget Parameters)))

(fn OnExecute
  (params (MyTarget :type Actor%20Object%20Reference) (Parameters :type Gameplay%20Cue%20Parameters%20Structure%20%28by%20ref%29))
  (return (|CallParentFunction MyTarget Parameters)))

(fn OnRemove
  (params (MyTarget :type Actor%20Object%20Reference) (Parameters :type Gameplay%20Cue%20Parameters%20Structure%20%28by%20ref%29))
  (return (|CallParentFunction MyTarget Parameters)))

(fn OnStartParticleSystemSpawned
  (params (SpawnedParticleSysComponent :type Cascade%20Particle%20System%20Component%20Object%20Reference) (Parameters :type Gameplay%20Cue%20Parameters%20Structure%20%28by%20ref%29))
  (bind (_normalizedmagnitude _rawmagnitude _effectcontext _matchedtagname _originaltag _aggregatedsourcetags _aggregatedtargettags _location _normal _instigator _effectcauser _sourceobject) (Ability|GameplayCue|BreakGameplayCueParameters Parameters))
  (bind _asfort_player_pawn_athena (Utilities|Casting|CastToFortPlayerPawnAthena _instigator))
  (if (Utilities|Equal(Object) _asfort_player_pawn_athena (Game|GetPlayerPawn 0))
    (bind _asathena_emoji_item_definition (Utilities|Casting|CastToAthenaEmojiItemDefinition _sourceobject))
    (ParticleSystem|ConfigureParticleSystem _asathena_emoji_item_definition SpawnedParticleSysComponent (Utilities|ResolveSoftReference (Variables|/Script/FortniteGame.FortPlayerPawnAthena|GetCurrentQuickChatIcon _asfort_player_pawn_athena)) false)
    (else
      ;[1 1 1] Added for Live Broadcast Spectator
      (bind _asfort_live_broadcast_controller (Utilities|Casting|CastToFortLiveBroadcastController (Game|GetPlayerController 0))
        (:then
          (if (Utilities|Equal(Object) _asfort_player_pawn_athena (Game|Player|GetPlayerPawn _asfort_live_broadcast_controller))
            (bind _asathena_emoji_item_definition_2 (Utilities|Casting|CastToAthenaEmojiItemDefinition _sourceobject)
              (:then
                (ParticleSystem|ConfigureParticleSystem _asathena_emoji_item_definition_2 SpawnedParticleSysComponent (Utilities|ResolveSoftReference (Variables|/Script/FortniteGame.FortPlayerPawnAthena|GetCurrentQuickChatIcon _asfort_player_pawn_athena)) false)
                (Transformation|SetRelativeLocation SpawnedParticleSysComponent (+ (Variables|/Script/Engine.SceneComponent|GetRelativeLocation SpawnedParticleSysComponent) (Variables|Default|GetParticleRelativeOffset)) false false)
                (return))
              (:CastFailed
                _)))
          (Components|Activation|Deactivate SpawnedParticleSysComponent))
        (:CastFailed
          (Components|Activation|Deactivate SpawnedParticleSysComponent))))))
